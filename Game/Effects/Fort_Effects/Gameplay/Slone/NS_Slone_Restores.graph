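# Niagara effect for Slone's restore moment: a short build-up and flare burst that tracks the skeletal mesh supplied through User.SloneMesh.
# GPU beam sprites follow the mesh with depth-buffer placement, while 77 decal sprites each track a bone index.
# Flare, corona, light and two opposing sweeping flares share one bone, where drifting mesh fragments and shimmer also spawn.
# Spawn groups separate layers within combined emitters, and color curves shift blue to pink.
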NiagaraSystem /Game/Effects/Fort_Effects/Gameplay/Slone/NS_Slone_Restores
dependencies {
 usedDataInterfaces [
  &NiagaraDataInterfaceCurve
  &NiagaraDataInterfaceColorCurve
  &NiagaraDataInterfaceArrayFloat
  &NiagaraDataInterfaceArrayColor
 ]
 usedDynamicInputs []
 usedModules [&SystemState]
}
emitters {
 Beam_Up_Build_Up {
  simTarget=GPUComputeSim
  emitterUpdateScript [
   &EmitterState()
   &SpawnRate(
    SpawnRate=&FloatFromCurve(
     FloatCurve=DataInterface() {propertyValues {Curve {keys [{arriveTangent=-2.8136, interpMode=RCIM_Cubic, leaveTangent=-2.8136, tangentMode=RCTM_User, time=0, value=1}, <at 1 0 RCIM_Cubic>]}}}
     CurveIndex=$Emitter.NormalizedLoopAge
     Scale Curve=154
    )
    Use Spawn Probability=false
   )
  ]
  particleSpawnScript [&InitializeParticle(
    Lifetime=&RandomRangeFloat(Minimum=0.2, Maximum=0.4)
    Sprite Size Mode=::NewEnumerator1
    Uniform Sprite Size Max=100
    Uniform Sprite Size Min=50
    UsePositionOffset=false
    Write Material Random=true
   )]
  particleUpdateScript [
   &ParticleState()
   &SkeletalMeshLocation(
    Bone Sampling Mode=::NewEnumerator0
    MaskSpawnGroup=false
    OverrideRandomSeed=false
    Position Sampling=::NewEnumerator1
    Sampled Position Offset Coordinate Space:ENiagaraSkelMeshTransforms=::NewEnumerator1
    Skeletal Mesh=$User.SloneMesh
   )
   /Niagara/Modules/Math/PlaceParticlesOnDepthBuffer_GPU(
    Max Z Depth=50
    Module.Use Max Z Depth=true
    Write To Intrinsic Variables:NiagaraBool=&SetBoolByFloatComparison(A=$Particles.NormalizedAge, B=0.7, Comparison Type=::NewEnumerator7)
   )
   &AccelerationForce(Acceleration=&Multiply_VectorByFloat(
     Vector=&RandomRangeVector(Minimum=<0 0 50>, Maximum=<0 0 50>)
     Float=&FloatFromCurve(
      FloatCurve=DataInterface() {propertyValues {Curve {keys [
          {arriveTangent=-3.5325, interpMode=RCIM_Cubic, leaveTangent=-3.5325, tangentMode=RCTM_User, time=0, value=1}
          {interpMode=RCIM_Cubic, leaveTangent=2.448, time=1, value=0}
         ]}}}
      Scale Curve=-10
     )
    ))
   &Drag(Drag=5, Use Linear Drag=true, Use Rotational Drag=false)
   &SolveForcesAndVelocity(Clamp Velocity=false, Limit Acceleration=false)
   &ScaleSpriteSize(Scale Factor=&MakeVector2D(
     X=&FloatFromCurve(FloatCurve=DataInterface() {propertyValues {Curve {keys [<at 0 1>,<at 1 0>]}}})
     Y=&FloatFromCurve(
      FloatCurve=DataInterface() {propertyValues {Curve {keys [
          {arriveTangent=-64.9259, interpMode=RCIM_Cubic, leaveTangent=-64.9259, tangentMode=RCTM_User, time=0, value=10}
          {interpMode=RCIM_Cubic, leaveTangent=29.9912, time=0.5, value=1}
         ]}}}
      Scale Curve=4
     )
    ))
   &CameraOffset(Camera Offset Amount=&FloatFromCurve(FloatCurve=DataInterface() {propertyValues {Curve {keys [<at 0 5>,<at 1 1>]}}}, Scale Curve=10))
   &ScaleColor(
    Scale Alpha=&FloatFromCurve(FloatCurve=DataInterface() {propertyValues {Curve {keys [<at 0.5 1 RCIM_Cubic>, {arriveTangent=-5.4338, interpMode=RCIM_Cubic, leaveTangent=-5.4338, tangentMode=RCTM_User, time=1, value=0}]}}})
    Scale RGB=&MakeVectorFromLinearColorRGB(Color=&ColorFromCurve(ColorCurve=DataInterface() {propertyValues {
        AlphaCurve {keys []}
        BlueCurve {keys [<at 0 1 RCIM_Cubic>,<at 0.75 0.6267 RCIM_Cubic>]}
        GreenCurve {keys [<at 0 0.5219 RCIM_Cubic>,<at 0.75 0.6267 RCIM_Cubic>]}
        RedCurve {keys [<at 0 0.2031 RCIM_Cubic>,<at 0.75 1 RCIM_Cubic>]}
       }}))
    ScaleA=true
    ScaleRGB=true
   )
  ]
  rendererClasses [&NiagaraSpriteRendererProperties]
  renderers [{rendererClass=&NiagaraSpriteRendererProperties}]
 }
 Combined_Flare_Corona_Emitter {
  emitterUpdateScript [
   &EmitterState()
   &SpawnBurst_Instantaneous(Spawn Count=1, Use Spawn Probability=false)
   &SpawnBurst_Instantaneous(Spawn Group=1, Use Spawn Probability=false, UseLoopCountLimit=false) {module=SpawnBurst_Instantaneous001}
   &SpawnBurst_Instantaneous(Spawn Group=2, Use Spawn Probability=false, UseLoopCountLimit=false) {module=SpawnBurst_Instantaneous002}
  ]
  particleSpawnScript [
   /SetVariables(Particles.VisibilityTag=&SpawnGroup()) {isSetParameters=true}
   &InitializeParticle(
    Color=&SelectLinearColorFromArray(
     Array Sampling Mode=::NewEnumerator1
     Direct Array Index=&SpawnGroup()
     Color Selection Array=DataInterface() {propertyValues {ColorData [<rgba 1 1 1 2>, <rgba 2 2 2 1>, <rgba 1 1 1 1>]}}
     Random Seed=?
    )
    Lifetime=&SelectFloatFromArray(
     Direct Array Index=&SpawnGroup()
     Float Selection Array=DataInterface() {propertyValues {FloatData [0.4,0.5,0.5]}}
    )
    Sprite Size Mode=::NewEnumerator3
    Uniform Sprite Size=&SelectFloatFromArray(
     Direct Array Index=&SpawnGroup()
     Float Selection Array=DataInterface() {propertyValues {FloatData [150,250,0]}}
    )
    UsePositionOffset=false
    Write Material Random=true
   )
   &CameraOffset(Camera Offset Amount=150)
  ]
  particleUpdateScript [
   &ParticleState()
   &SkeletalMeshLocation(
    Bone / Socket Index=1
    Bone Sampling Mode=::NewEnumerator4
    MaskSpawnGroup=false
    Sampled Position Offset Coordinate Space:ENiagaraSkelMeshTransforms=::NewEnumerator2
    Skeletal Mesh=$User.SloneMesh
   )
   &ScaleColor(
    Scale Alpha=&FloatFromCurve(
     FloatCurve=DataInterface() {propertyValues {Curve {keys [{arriveTangent=2.4184, interpMode=RCIM_Cubic, leaveTangent=2.4184, tangentMode=RCTM_User, time=0.25, value=0}, <at 1 1 RCIM_Cubic>]}}}
     Scale Curve=2
    )
    Scale RGB=&MakeVectorFromLinearColorRGB(Color=&ColorFromCurve(ColorCurve=DataInterface() {propertyValues {
        AlphaCurve {keys []}
        BlueCurve {keys [<at 0.25 1 RCIM_Cubic>,<at 1 0.6267 RCIM_Cubic>]}
        GreenCurve {keys [<at 0.25 0.5219 RCIM_Cubic>,<at 1 0.6267 RCIM_Cubic>]}
        RedCurve {keys [<at 0.25 0.2031 RCIM_Cubic>,<at 1 1 RCIM_Cubic>]}
       }}))
    ScaleA=true
    ScaleRGB=true
   )
   &ScaleSpriteSize(Scale Factor=&Vector2DFromFloat(Value=&FloatFromCurve(FloatCurve=DataInterface() {propertyValues {Curve {keys [<at 0 1>,<at 1 5>]}}})))
   /SetVariables(
    Particles.Corona_Color:LinearColor=&MultiplyLinearColor(
     Linear Color=$Particles.Initial.Color
     Scale Factor=&MakeFloat4FromVectorAndFloat(
      "Vector (RGB)"=&MakeVectorFromLinearColorRGB(Color=&ColorFromCurve(ColorCurve=DataInterface() {propertyValues {
          AlphaCurve {keys []}
          BlueCurve {keys [<at 0.25 1 RCIM_Cubic>,<at 0.75 0.6267 RCIM_Cubic>]}
          GreenCurve {keys [<at 0.25 0.5219 RCIM_Cubic>,<at 0.75 0.6267 RCIM_Cubic>]}
          RedCurve {keys [<at 0.25 0.2031 RCIM_Cubic>,<at 0.75 1 RCIM_Cubic>]}
         }}))
      "Float (Alpha)"=&FloatFromCurve(FloatCurve=DataInterface() {propertyValues {Curve {keys [
           {arriveTangent=25.6036, interpMode=RCIM_Cubic, leaveTangent=25.6036, tangentMode=RCTM_User, time=0, value=0}
           {arriveTangentWeight=0.0556, interpMode=RCIM_Cubic, leaveTangentWeight=0.3611, tangentMode=RCTM_User, time=0.1, value=1}
           {arriveTangentWeight=2.0573, interpMode=RCIM_Cubic, leaveTangentWeight=0.7913, tangentMode=RCTM_User, time=0.75, value=1}
           <at 1 0 RCIM_Cubic>
          ]}}})
     )
    )
    Particles.Corona_Dynamic:Vector4f=&MakeVector4(
     X=&FloatFromCurve(FloatCurve=DataInterface() {propertyValues {Curve {keys [
          {arriveTangent=7.0871, interpMode=RCIM_Cubic, leaveTangent=7.0871, tangentMode=RCTM_User, time=0, value=0}
          {interpMode=RCIM_Cubic, leaveTangent=-6.606, time=0.4, value=1}
         ]}}})
     Y=1
     Z=0.25
     W=1
    )
    Particles.Corona_Size:Vector2f=&Multiply_Vector2D(
     A=$Particles.Initial.SpriteSize
     B=&Vector2DFromFloat(Value=&FloatFromCurve(FloatCurve=DataInterface() {propertyValues {Curve {keys [
           {arriveTangent=15.4838, interpMode=RCIM_Cubic, time=0, value=2}
           {arriveTangent=-12.4294, interpMode=RCIM_Cubic, leaveTangent=-12.4294, tangentMode=RCTM_User, time=0.2, value=1}
          ]}}}))
    )
   ) {isSetParameters=true}
   /SetVariables(Particles.LightColor=&MultiplyLinearColor(
     Linear Color=$Particles.Initial.Color
     Scale Factor=&MakeFloat4FromVectorAndFloat(
      "Vector (RGB)"=&MakeVectorFromLinearColorRGB(Color=&ColorFromCurve(ColorCurve=DataInterface() {propertyValues {
          AlphaCurve {keys []}
          BlueCurve {keys [<at 0.25 1 RCIM_Cubic>,<at 0.75 0.6267 RCIM_Cubic>]}
          GreenCurve {keys [<at 0.25 0.5219 RCIM_Cubic>,<at 0.75 0.6267 RCIM_Cubic>]}
          RedCurve {keys [<at 0.25 0.2031 RCIM_Cubic>,<at 0.75 1 RCIM_Cubic>]}
         }}))
      "Float (Alpha)"=1
     )
    )) {isSetParameters=true}
   &FN_Light_Attributes(
    Color=$Particles.LightColor
    Create Color Attribute=true
    Create Exponent Attribute=true
    Create Radius Attribute=true
    Exponent=3
    Radius=150
    Scale Color=&FloatFromCurve(
     FloatCurve=DataInterface() {propertyValues {Curve {keys [<at 0.5 1 RCIM_Cubic>, {arriveTangent=-3.7907, interpMode=RCIM_Cubic, leaveTangent=-3.7907, tangentMode=RCTM_User, time=1, value=0}]}}}
     Scale Curve=150
    )
   )
  ]
  rendererClasses [&NiagaraSpriteRendererProperties, &NiagaraLightRendererProperties]
  renderers [
   {rendererClass=&NiagaraSpriteRendererProperties}
   {rendererClass=&NiagaraSpriteRendererProperties, rendererIndex=1}
   {rendererClass=&NiagaraLightRendererProperties, rendererIndex=2}
  ]
 }
 Combined_Tiny_Frags_Shimmer {
  simTarget=GPUComputeSim
  emitterUpdateScript [
   &EmitterState()
   &SpawnBurst_Instantaneous(Spawn Count=&RandomRangeInt(Minimum=25, Maximum=75), Use Spawn Probability=false)
   &SpawnBurst_Instantaneous(Spawn Count=&RandomRangeInt(Minimum=3, Maximum=5), Spawn Group=1, Use Spawn Probability=false) {module=SpawnBurst_Instantaneous001}
  ]
  particleSpawnScript [
   /SetVariables(Particles.VisibilityTag=&SpawnGroup()) {isSetParameters=true}
   &InitializeParticle(
    Color=&MakeLinearColorFromVectorAndFloat(
     "Vector (RGB)"=&MaskVectorBySpawnGroup(
      Masked Vector=&MakeVectorFromLinearColorRGB(Color=<rgba 0.605 0.6866 0.7396 1>)
      Passthrough Vector=&MakeVectorFromLinearColorRGB(Color=<rgba 10 10 10 1>)
      Spawn Group Mask=1
     )
     "Float (Alpha)"=&RandomRangeFloat(
      Minimum=&MaskFloatBySpawnGroup(Masked Float=0.2, Passthrough Float=1, Spawn Group Mask=1)
      Maximum=&MaskFloatBySpawnGroup(Masked Float=0.5, Passthrough Float=1, Spawn Group Mask=1)
     )
    )
    Lifetime Max=&MaskFloatBySpawnGroup(Masked Float=2, Passthrough Float=1.25, Spawn Group Mask=1)
    Lifetime Mode=::NewEnumerator1
    Mesh Scale Max=<0.7 0.7 1>
    Mesh Scale Min=<0.4 0.4 1>
    Mesh Scale Mode=::NewEnumerator2
    Sprite Size Mode=::NewEnumerator1
    Uniform Sprite Size Max=250
    Uniform Sprite Size Min=225
    UsePositionOffset=false
    Write Material Random=true
   )
   &SkeletalMeshLocation(
    Bone / Socket Index=1
    Bone Sampling Mode=::NewEnumerator4
    MaskSpawnGroup=false
    Sampled Position Offset Coordinate Space:ENiagaraSkelMeshTransforms=::NewEnumerator2
    Skeletal Mesh=$User.SloneMesh
   )
   &SphereLocation(
    MaskSpawnGroup=false
    Override Seed=false
    Sphere Radius=&MaskFloatBySpawnGroup(Masked Float=25, Passthrough Float=75, Spawn Group Mask=1)
    Surface Only=false
    UseNonUniformScale=false
    UseOffset=false
    UseSphereOrientation=false
   )
   &InitialMeshOrientation(
    Rotation=&RandomRangeVector(Minimum=<0 0 0>, Maximum=<1 1 1>)
    Use Orientation Vector=false
    Use Rotation Vector=true
   )
   /SetVariables(Particles.Shimmer Dynamic:Vector4f=&MakeVector4(
     X=&RandomRangeFloat(Minimum=0, Maximum=1)
     Y=&RandomRangeFloat(Minimum=0, Maximum=1)
     Z=0
     W=0
    )) {isSetParameters=true}
  ]
  particleUpdateScript [
   &ParticleState()
   &PointForce(
    Force Falloff Distance=175
    Force Strength=&RandomRangeFloat(Minimum=33, Maximum=66)
    Linear Force Vector:Vector3f=<0 0 1>
    Module.Use Falloff=true
    Module.UseLinearForce=true
   )
   &CurlNoiseForce(Noise Frequency=30, Noise Strength=100, Pan Noise=true, Pan Noise Field=<0 0 -5>)
   &SolveForcesAndVelocity(Clamp Velocity=false, Limit Acceleration=false)
   &UpdateMeshOrientation(Rotation Rate=0.15, Rotation Vector=&RandomRangeVector(Minimum=<0 0 0>, Maximum=<1 1 1>))
   &ScaleMeshSize(Scale Factor=&VectorFromFloat(Value=&FloatFromCurve(FloatCurve=DataInterface() {propertyValues {Curve {keys [
          {arriveTangent=12.2517, interpMode=RCIM_Cubic, leaveTangent=12.2517, tangentMode=RCTM_User, time=0, value=0}
          {arriveTangentWeight=0.1067, interpMode=RCIM_Cubic, leaveTangentWeight=0.3202, tangentMode=RCTM_User, time=0.2, value=1}
          {arriveTangentWeight=0.3202, interpMode=RCIM_Cubic, leaveTangentWeight=0.1067, tangentMode=RCTM_User, time=0.8, value=1}
          <at 1 0 RCIM_Cubic>
         ]}}})))
   /SetVariables(
    Particles.Shimmer_Camera:NiagaraFloat=&FloatFromCurve(FloatCurve=DataInterface() {propertyValues {Curve {keys [<at 0 0.5>,<at 1 1.25>]}}}, Scale Curve=50)
    Particles.Shimmer_Color:LinearColor=&MultiplyLinearColor(
     Linear Color=$Particles.Initial.Color
     Scale Factor=&MakeFloat4FromVectorAndFloat(
      "Vector (RGB)"=<1 1 1>
      "Float (Alpha)"=&FloatFromCurve(FloatCurve=DataInterface() {propertyValues {Curve {keys [
           {interpMode=RCIM_Cubic, tangentMode=RCTM_User, time=0, value=0}
           {interpMode=RCIM_Cubic, tangentMode=RCTM_User, time=0.5, value=1}
           {tangentMode=RCTM_User, time=1, value=0}
          ]}}})
     )
    )
   ) {isSetParameters=true}
  ]
  rendererClasses [&NiagaraMeshRendererProperties, &NiagaraSpriteRendererProperties]
  renderers [
   {rendererClass=&NiagaraMeshRendererProperties, rendererIndex=0}
   {rendererClass=&NiagaraSpriteRendererProperties, rendererIndex=1}
  ]
 }
 Decal_Build_Up {
  simTarget=GPUComputeSim
  emitterUpdateScript [
   &EmitterState()
   &SpawnBurst_Instantaneous(Spawn Count=77, Use Spawn Probability=false)
  ]
  particleSpawnScript [
   &FN_SetSpawnGroup(Number of Spawn Groups=77)
   &InitializeParticle(
    Color=<rgba 1 1 1 0.5>
    Lifetime=0.349
    Sprite Size Mode=::NewEnumerator3
    Uniform Sprite Size=50
    UsePositionOffset=false
    Write Material Random=true
   )
  ]
  particleUpdateScript [
   &ParticleState()
   &SkeletalMeshLocation(Bone / Socket Index=$Particles.SpawnGroup, Bone Sampling Mode=::NewEnumerator4, MaskSpawnGroup=false, Skeletal Mesh=$User.SloneMesh)
   &CameraOffset(Camera Offset Amount=25)
   &ScaleColor(
    Scale Alpha=&FloatFromCurve(
     FloatCurve=DataInterface() {propertyValues {Curve {keys [
         {arriveTangent=2.4184, interpMode=RCIM_Cubic, time=0, value=10}
         {arriveTangent=-3.2377, interpMode=RCIM_Cubic, leaveTangent=-3.2377, tangentMode=RCTM_User, time=0.75, value=0}
        ]}}}
     Scale Curve=2
    )
    Scale RGB=&MakeVectorFromLinearColorRGB(Color=&ColorFromCurve(ColorCurve=DataInterface() {propertyValues {
        AlphaCurve {keys []}
        BlueCurve {keys [<at 0.25 1 RCIM_Cubic>,<at 0.75 0.6267 RCIM_Cubic>]}
        GreenCurve {keys [<at 0.25 0.5219 RCIM_Cubic>,<at 0.75 0.6267 RCIM_Cubic>]}
        RedCurve {keys [<at 0.25 0.2031 RCIM_Cubic>,<at 0.75 1 RCIM_Cubic>]}
       }}))
    ScaleA=true
    ScaleRGB=true
   )
   &DynamicMaterialParameters(
    Index 0 Param 1=0.5
    Index 0 Param 2=1
    Index 0 Param 3=&RandomRangeFloat(Minimum=0, Maximum=1)
    Index 0 Param 4=&RandomRangeFloat(Minimum=0, Maximum=1)
    Index 1 Param 1=25
    Index 1 Param 2=20
    Module.Param0WriteEnabled=true
    Module.Param0WriteEnabled1=true
    Module.Param1WriteEnabled=true
    Module.Param1WriteEnabled1=true
    Module.Param2WriteEnabled=true
    Module.Param2WriteEnabled1=true
    Module.Param3WriteEnabled=true
    Module.Param3WriteEnabled1=true
    Write Parameter Index 1=true
   )
  ]
  rendererClasses [&NiagaraSpriteRendererProperties]
  renderers [{rendererClass=&NiagaraSpriteRendererProperties}]
 }
 Sweeping_Flares {
  emitterUpdateScript [
   &EmitterState()
   &SpawnBurst_Instantaneous(Spawn Count=1, Use Spawn Probability=false)
   &SpawnBurst_Instantaneous(Spawn Count=1, Spawn Group=1, Use Spawn Probability=false) {module=SpawnBurst_Instantaneous001}
  ]
  particleSpawnScript [
   &InitializeParticle(
    Color Mode=::NewEnumerator0
    Lifetime=0.5
    Sprite Size=<50 200>
    Sprite Size Mode=::NewEnumerator4
    UsePositionOffset=false
    Write Material Random=true
   )
   &CameraOffset(Camera Offset Amount=150)
  ]
  particleUpdateScript [
   &ParticleState()
   /SetVariables(Particles.BeamOffset:Vector3f=&MakeVector(
     X=&FloatFromCurve(
      FloatCurve=DataInterface() {propertyValues {Curve {keys [
          {arriveTangent=-2.0552, interpMode=RCIM_Cubic, time=0, value=1}
          {arriveTangent=-2.0672, interpMode=RCIM_Cubic, leaveTangent=-2.0672, tangentMode=RCTM_User, time=1, value=0}
         ]}}}
      Scale Curve=75
     )
     Y=0
     Z=0
    )) {isSetParameters=true}
   &SkeletalMeshLocation(
    Bone / Socket Index=1
    Bone Sampling Mode=::NewEnumerator4
    MaskSpawnGroup=false
    Sampled Position Offset=&MaskVectorBySpawnGroup(
     Masked Vector=&Multiply_VectorByFloat(Vector=$Particles.BeamOffset, Float=-1)
     Passthrough Vector=$Particles.BeamOffset
    )
    Sampled Position Offset Coordinate Space:ENiagaraSkelMeshTransforms=::NewEnumerator2
    Skeletal Mesh=$User.SloneMesh
   )
   &ScaleColor(
    Scale Alpha=&FloatFromCurve(
     FloatCurve=DataInterface() {propertyValues {Curve {keys [{arriveTangent=2.4184, interpMode=RCIM_Cubic, leaveTangent=2.4184, tangentMode=RCTM_User, time=0.25, value=0}, <at 1 1 RCIM_Cubic>]}}}
     Scale Curve=2
    )
    Scale RGB=&MakeVectorFromLinearColorRGB(Color=&ColorFromCurve(ColorCurve=DataInterface() {propertyValues {
        AlphaCurve {keys []}
        BlueCurve {keys [<at 0.25 1 RCIM_Cubic>,<at 0.75 0.6267 RCIM_Cubic>]}
        GreenCurve {keys [<at 0.25 0.5219 RCIM_Cubic>,<at 0.75 0.6267 RCIM_Cubic>]}
        RedCurve {keys [<at 0.25 0.2031 RCIM_Cubic>,<at 0.75 1 RCIM_Cubic>]}
       }}))
    ScaleA=true
    ScaleRGB=true
   )
   &ScaleSpriteSize(Scale Factor=&Vector2DFromFloat(Value=&FloatFromCurve(FloatCurve=DataInterface() {propertyValues {Curve {keys [<at 0 5>,<at 1 1>]}}})))
  ]
  rendererClasses [&NiagaraSpriteRendererProperties]
  renderers [{rendererClass=&NiagaraSpriteRendererProperties}]
 }
}
summary {
 systemName=NS_Slone_Restores
 emitters [
  {emitterName=Beam_Up_Build_Up, simTarget=GPUComputeSim, rendererClasses [&NiagaraSpriteRendererProperties]}
  {emitterName=Decal_Build_Up, simTarget=GPUComputeSim, rendererClasses [&NiagaraSpriteRendererProperties]}
  {emitterName=Sweeping_Flares, rendererClasses [&NiagaraSpriteRendererProperties]}
  {
   emitterName=Combined_Flare_Corona_Emitter
   rendererClasses [&NiagaraSpriteRendererProperties, &NiagaraLightRendererProperties]
  }
  {
   emitterName=Combined_Tiny_Frags_Shimmer
   simTarget=GPUComputeSim
   rendererClasses [&NiagaraMeshRendererProperties, &NiagaraSpriteRendererProperties]
  }
 ]
 userVariables [User.SloneMesh {
   type=NiagaraDataInterfaceSkeletalMesh
   defaultValue {
    struct=NiagaraExt_VariableValue_DataInterface
    value {dataInterface=&self:NiagaraDataInterfaceSkeletalMesh_0}
   }
  }]
}
